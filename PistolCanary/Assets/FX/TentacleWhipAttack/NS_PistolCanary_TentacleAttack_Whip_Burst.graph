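# Niagara burst effect for the PistolCanary tentacle whip attack, combining a brief glowing flash with flying sparks, sized by User.Scale.
# Burst layers four sprite renderers on a camera-offset particle, computing anamorphic streak, big glow, soft glow and waveform-flickering corona values, with several alphas attenuated by System.TimeForDis curves.
# BurstSparks launches randomized pink sparks from a sphere under drag and gravity, stretching them by speed and shrinking them over life.
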NiagaraSystem /PistolCanary/Assets/FX/TentacleWhipAttack/NS_PistolCanary_TentacleAttack_Whip_Burst
dependencies {usedDataInterfaces [&NiagaraDataInterfaceCurve], usedDynamicInputs [], usedModules [&SystemState]}
emitters {
 Burst {
  emitterUpdateScript [
   &EmitterState()
   &SpawnBurst_Instantaneous(Use Spawn Probability=false, UseLoopCountLimit=false)
  ]
  particleSpawnScript [
   &InitializeParticle(
    Color=<rgba 0.2217 0.6056 1.5 1>
    Lifetime=0.15
    Sprite Size Mode=::NewEnumerator3
    Uniform Sprite Size=&Multiply_Float(A=6250, B=$User.Scale)
    UsePositionOffset=false
    Write Material Random=true
   )
   &CameraOffset(Camera Offset Amount=1200) {module=CameraOffset001}
   /SetVariables(
    Particles.Animorphic Color:LinearColor=<rgba 1.5175 0 5 0.05>
    Particles.Animorphic Rotation=90
    Particles.Animorphic Size:Vector2f=&Multiply_Vector2D_ByFloat(Float=$User.Scale, Vector2D=<6000 150>)
    Particles.Animorphic Vector:Vector3f=<0 0 1>
    Particles.BigGlowSize:Vector2f=&Multiply_Vector2D_ByFloat(Float=$User.Scale, Vector2D=<5600 2912.2019>)
    Particles.Soft Glow Size:Vector2f=&Multiply_Vector2D_ByFloat(Float=$User.Scale, Vector2D=<6500 771.1771>)
   ) {isSetParameters=true}
  ]
  particleUpdateScript [
   &ParticleState()
   &ScaleColor(
    Scale Alpha=&FloatFromCurve(FloatCurve=DataInterface() {propertyValues {Curve {keys [
         {interpMode=RCIM_Cubic, tangentMode=RCTM_User, time=0, value=0}
         {interpMode=RCIM_Cubic, tangentMode=RCTM_User, time=0.5, value=1}
         {tangentMode=RCTM_User, time=1, value=0}
        ]}}})
    ScaleA=true
    ScaleRGB=true
   )
   /SetVariables(Particles.Soft Glow Color:LinearColor=&MakeLinearColorFromVectorAndFloat(
     "Vector (RGB)"=&MakeVectorFromLinearColorRGB(Color=<rgba 0 2.849 5 1>)
     "Float (Alpha)"=&FloatFromCurve(
      FloatCurve=DataInterface() {propertyValues {Curve {keys [
          {interpMode=RCIM_Cubic, tangentMode=RCTM_User, time=0, value=0}
          {interpMode=RCIM_Cubic, tangentMode=RCTM_User, time=0.5, value=1}
          {tangentMode=RCTM_User, time=1, value=0}
         ]}}}
      Scale Curve=&Multiply_Float(
       A=0.005
       B=&FloatFromCurve(
        FloatCurve=DataInterface() {propertyValues {Curve {keys [{interpMode=RCIM_Cubic, leaveTangent=-3.7201, tangentMode=RCTM_Break, time=0, value=1}, <at 1 0>]}}}
        CurveIndex=$System.TimeForDis
       )
      )
     )
    )) {isSetParameters=true}
   /SetVariables(
    Particles.Corona Color:LinearColor=&MakeLinearColorFromVectorAndFloat(
     "Vector (RGB)"=&MakeVectorFromLinearColorRGB(Color=<rgba 1 0.0889 0.1786 1>)
     "Float (Alpha)"=&FloatFromCurve(
      FloatCurve=DataInterface() {propertyValues {Curve {keys [
          <at 0 0 RCIM_Cubic>
          {arriveTangent=2.5, interpMode=RCIM_Cubic, leaveTangent=2.5, time=0.165, value=0.65}
          {arriveTangent=0.1786, interpMode=RCIM_Cubic, leaveTangent=0.1786, time=0.4, value=1}
          {arriveTangent=-1.6667, interpMode=RCIM_Cubic, leaveTangent=-1.6667, time=0.725, value=0.75}
          <at 1 0 RCIM_Cubic>
         ]}}}
      Scale Curve=&Multiply_Float(
       A=0.5
       B=&FloatFromCurve(
        FloatCurve=DataInterface() {propertyValues {Curve {keys [{interpMode=RCIM_Cubic, leaveTangent=-3.7201, tangentMode=RCTM_Break, time=0, value=1}, <at 1 0>]}}}
        CurveIndex=$System.TimeForDis
       )
      )
     )
    )
    Particles.Corona Size:Vector2f=&Vector2DFromFloat(Value=&FloatFromCurve(
      FloatCurve=DataInterface() {propertyValues {Curve {keys [<at 0 1>,<at 0.5 0.7>,<at 1 1>]}}}
      Scale Curve=&Waveform(
       Global Amplitude Scale=&Multiply_Float(A=2000, B=$User.Scale)
       "Waveform [1]"=::NewEnumerator12
       "[1] Phase Scale"=50
       "[1] Frequency"=1
       "[1] Amplitude Min/Max"=<0.85 1>
      )
     ))
   ) {isSetParameters=true}
   &DynamicMaterialParameters(
    Index 3 Param 1=&FloatFromCurve(FloatCurve=DataInterface() {propertyValues {Curve {keys [<at 0 0.7 RCIM_Cubic>,<at 0.5 0.25 RCIM_Cubic>,<at 1 0.7 RCIM_Cubic>]}}})
    Index 3 Param 2=1
    Index 3 Param 3=&RandomRangeFloat(Minimum=0.25, Maximum=0.35)
    Index 3 Param 4=1
    Module.Param0WriteEnabled3=true
    Module.Param1WriteEnabled3=true
    Module.Param2WriteEnabled3=true
    Module.Param3WriteEnabled3=true
    Write Parameter Index 0=false
    Write Parameter Index 3=true
   )
   /SetVariables() {isSetParameters=true}
   &ScaleColor(
    Scale Alpha=&FloatFromCurve(
     FloatCurve=DataInterface() {propertyValues {Curve {keys [{interpMode=RCIM_Cubic, leaveTangent=-3.7201, tangentMode=RCTM_Break, time=0, value=1}, <at 1 0>]}}}
     CurveIndex=$System.TimeForDis
    )
    ScaleA=true
    ScaleRGB=true
   ) {module=ScaleColor001}
   /SetVariables(Particles.Animorphic Color:LinearColor=&MultiplyLinearColor(
     Linear Color=$Particles.Initial.Animorphic Color
     Scale Factor=&MakeVector4FromVectorAndFloat(
      "Vector (RGB)"=<1 1 1>
      "Float (Alpha)"=&FloatFromCurve(
       FloatCurve=DataInterface() {propertyValues {Curve {keys [{interpMode=RCIM_Cubic, leaveTangent=-3.7201, tangentMode=RCTM_Break, time=0, value=1}, <at 1 0>]}}}
       CurveIndex=$System.TimeForDis
      )
     )
    )) {isSetParameters=true}
   /SetVariables(Particles.BigGlowColor:LinearColor=&MultiplyLinearColor(
     Linear Color=$Particles.Animorphic Color
     Scale Factor=&MakeVector4FromVectorAndFloat(
      "Vector (RGB)"=<0.5 0.5 0.5>
      "Float (Alpha)"=&FloatFromCurve(FloatCurve=DataInterface() {propertyValues {Curve {
          defaultValue=1
          keys [
           {arriveTangent=9.0725, interpMode=RCIM_Cubic, leaveTangent=9.0725, tangentMode=RCTM_User, time=0, value=0}
           {arriveTangent=-2.5249, interpMode=RCIM_Cubic, leaveTangent=-2.5249, tangentMode=RCTM_User, time=0.1408, value=0.9847}
           {arriveTangent=-0.0405, interpMode=RCIM_Cubic, leaveTangent=-0.0405, tangentMode=RCTM_User, time=1, value=0}
          ]
         }}})
     )
    )) {isSetParameters=true}
  ]
  rendererClasses [&NiagaraSpriteRendererProperties]
  renderers [
   {rendererClass=&NiagaraSpriteRendererProperties}
   {rendererClass=&NiagaraSpriteRendererProperties, rendererIndex=1}
   {rendererClass=&NiagaraSpriteRendererProperties, rendererIndex=2}
   {rendererClass=&NiagaraSpriteRendererProperties, rendererIndex=3}
  ]
 }
 BurstSparks {
  emitterUpdateScript [
   &EmitterState()
   &SpawnBurst_Instantaneous(
    Spawn Count=&RandomRangeInt(Minimum=50, Maximum=60)
    Use Spawn Probability=false
    UseLoopCountLimit=false
   )
  ]
  particleSpawnScript [
   &InitializeParticle(
    Color=<rgba 275 65.3025 230.0222 1>
    Lifetime Max=0.75
    Lifetime Min=0.3
    Lifetime Mode=::NewEnumerator1
    Sprite Size=&Multiply_Vector2D_ByFloat(
     Float=&Multiply_Float(A=&RandomRangeFloat(Minimum=0.5, Maximum=2), B=$User.Scale)
     Vector2D=<50 100>
    )
    Sprite Size Mode=::NewEnumerator4
    UsePositionOffset=false
    Write Material Random=true
   )
   &ShapeLocation(
    MaskSpawnGroup=false
    Override Seed=false
    Sphere Radius=&Multiply_Float(A=&RandomRangeFloat(Minimum=5, Maximum=100), B=$User.Scale)
    UseOwnerScale=true
   )
   &AddVelocity(
    Module.Use Falloff Distance=false
    Override Seed=false
    Velocity Mode=::NewEnumerator1
    Velocity Speed=&Multiply_Float(A=&RandomRangeFloat(Minimum=250, Maximum=4600), B=$User.Scale)
   )
  ]
  particleUpdateScript [
   &ParticleState()
   &Drag(Use Linear Drag=true, Use Rotational Drag=true)
   &GravityForce()
   &SolveForcesAndVelocity(Clamp Velocity=false, Limit Acceleration=false)
   &ScaleSpriteSizeBySpeed(Max Scale Factor=<1 2>, Min Scale Factor=<0.5 1>, Velocity Threshold=12000)
   &ScaleSpriteSize(Uniform Curve Sprite Scale=DataInterface() {propertyValues {Curve {
       defaultValue=0
       keys [{interpMode=RCIM_Cubic, tangentMode=RCTM_User, time=0, value=1}, {interpMode=RCIM_Cubic, tangentMode=RCTM_User, time=1, value=0}]
      }}})
  ]
  rendererClasses [&NiagaraSpriteRendererProperties]
  renderers [{rendererClass=&NiagaraSpriteRendererProperties}]
 }
}
summary {
 systemName=NS_PistolCanary_TentacleAttack_Whip_Burst
 emitters [
  {emitterName=Burst, rendererClasses [&NiagaraSpriteRendererProperties]}
  {emitterName=BurstSparks, rendererClasses [&NiagaraSpriteRendererProperties]}
 ]
 userVariables [
  User.BeamEnd_0 {type=NiagaraPosition, description="", defaultValue {}}
  User.Scale {type=NiagaraFloat, defaultValue=1, description=""}
 ]
}
