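# Niagara projectile effect for PistolCanary's tentacle whip attack: a purple-blue energy orb with crackle and light.
# Persistent single-burst sprites form the core flare, glow with dynamic light, and spinning radial energy, plus a waveform-pulsed PostRadialBlur material sprite.
# Continuous emitters add velocity-stretched sparks, flickering corona flares and short-lived lightning meshes.
# User.Scale drives most sizes, User.PrimaryColor tints only the inner flares, and the charge-driven BigBall mesh is disabled.
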NiagaraSystem /PistolCanary/Assets/FX/TentacleWhipAttack/NS_PistolCanary_TentacleAttack_Whip_Proj
dependencies {usedDataInterfaces [&NiagaraDataInterfaceCurve], usedDynamicInputs [], usedModules [&SystemState]}
emitters {
 BigBall {
  bEnabled=false
  emitterUpdateScript [
   &EmitterState(
    Inactive Response=::NewEnumerator2
    Life Cycle Mode=::NewEnumerator1
    Loop Behavior=::NewEnumerator1
    Loop Duration Mode=::NewEnumerator1
    UseLoopDelay=false
   )
   &SpawnBurst_Instantaneous(Spawn Count=1, Use Spawn Probability=false)
  ]
  particleSpawnScript [
   &InitializeParticle(
    Color=<rgba 5 0.6339 4.2804 0.35>
    Lifetime=$System.CurrentLoopDuration
    Mesh Scale Mode=::NewEnumerator3
    Mesh Uniform Scale=7
    UsePositionOffset=false
    Write Material Random=true
   )
   &InitialMeshOrientation(Orientation Vector=<0 0 1>, Rotation=<0 0.5 0>, Use Orientation Vector=true, Use Rotation Vector=true)
  ]
  particleUpdateScript [
   &ParticleState(Kill Particles When Lifetime Has Elapsed=false)
   &ScaleMeshSize(Scale Factor=&Multiply_VectorByFloat(Vector=<1 1 1>, Float=$System.ChargeProgress))
   &ScaleColor(
    Scale Alpha=&FloatFromCurve(
     FloatCurve=DataInterface() {propertyValues {Curve {
        defaultValue=1
        keys [{interpMode=RCIM_Cubic, tangentMode=RCTM_User, time=0.65, value=0}, {interpMode=RCIM_Cubic, tangentMode=RCTM_User, time=1, value=1}]
       }}}
     CurveIndex=$System.ChargeProgress
    )
    ScaleA=true
    ScaleRGB=true
   )
  ]
  rendererClasses [&NiagaraMeshRendererProperties]
  renderers [{rendererClass=&NiagaraMeshRendererProperties, rendererIndex=0}]
 }
 CoreFlare {
  emitterUpdateScript [
   &EmitterState(
    Inactive Response=::NewEnumerator2
    Life Cycle Mode=::NewEnumerator1
    Loop Behavior=::NewEnumerator1
    Loop Duration Mode=::NewEnumerator1
    UseLoopDelay=false
   ) {module=EmitterState001}
   &SpawnBurst_Instantaneous(Spawn Count=1, Use Spawn Probability=false)
  ]
  particleSpawnScript [&InitializeParticle(
    Color Mode=::NewEnumerator0
    Sprite Rotation Mode=::NewEnumerator2
    Sprite Size Mode=::NewEnumerator3
    Uniform Sprite Size=&Multiply_Float(A=3125, B=$User.Scale)
    UsePositionOffset=false
    Write Material Random=true
   )]
  particleUpdateScript [
   &ParticleState(Kill Particles When Lifetime Has Elapsed=false)
   &Color(Color=<rgba 5.6001 0 20 1>)
   /SetVariables(
    Particles.SpriteRotation:NiagaraFloat=&RandomRangeFloat(Minimum=0, Maximum=359, Evaluation Type=::NewEnumerator1)
    Particles.Vector 2D:Vector2f=&Multiply_Vector2D_ByFloat(Float=1, Vector2D=$Particles.SpriteSize)
   ) {isSetParameters=true}
  ]
  rendererClasses [&NiagaraSpriteRendererProperties]
  renderers [{rendererClass=&NiagaraSpriteRendererProperties}]
 }
 Glow {
  emitterUpdateScript [
   &EmitterState(
    Inactive Response=::NewEnumerator2
    Life Cycle Mode=::NewEnumerator1
    Loop Behavior=::NewEnumerator1
    Loop Duration Mode=::NewEnumerator1
    UseLoopDelay=false
   ) {module=EmitterState001}
   &SpawnBurst_Instantaneous(Spawn Count=1, Use Spawn Probability=false)
  ]
  particleSpawnScript [
   &InitializeParticle(
    Lifetime=10
    Sprite Size Mode=::NewEnumerator3
    Uniform Sprite Size=&Multiply_Float(A=1500, B=$User.Scale)
    UsePositionOffset=false
    Write Material Random=true
   )
   &DynamicMaterialParameters(
    Index 0 Param 1=1
    Index 0 Param 2=1
    Index 0 Param 3=1
    Index 0 Param 4=1
    Module.Param0WriteEnabled=true
    Module.Param1WriteEnabled=true
    Module.Param2WriteEnabled=true
    Module.Param3WriteEnabled=true
   ) {module=DynamicMaterialParameters001}
  ]
  particleUpdateScript [
   &ParticleState(Kill Particles When Lifetime Has Elapsed=false)
   /SetVariables(Particles.Glow Camera offset=100) {isSetParameters=true}
   /SetVariables(Particles.Glow Color=<rgba 0.193 0.0836 0.3 0.5>) {isSetParameters=true}
   &FN_Light_Attributes(
    Color=<rgba 0.7996 0 5 1>
    Create Color Attribute=true
    Create Exponent Attribute=true
    Create Position Attribute=true
    Create Radius Attribute=true
    Exponent=2
    Radius=4500
    Scale Color=20
   )
  ]
  rendererClasses [&NiagaraSpriteRendererProperties, &NiagaraLightRendererProperties]
  renderers [
   {rendererClass=&NiagaraSpriteRendererProperties}
   {rendererClass=&NiagaraLightRendererProperties, rendererIndex=1}
   {rendererClass=&NiagaraSpriteRendererProperties, rendererIndex=2}
  ]
 }
 LoopSparks {
  emitterUpdateScript [
   &EmitterState(Life Cycle Mode=::NewEnumerator1, Loop Duration=40, UseLoopDelay=false)
   &SpawnRate(
    SpawnRate=&FloatFromCurve(
     FloatCurve=DataInterface() {propertyValues {Curve {keys [
         {interpMode=RCIM_Cubic, leaveTangent=0.3372, tangentMode=RCTM_Break, time=3.9, value=0}
         {arriveTangent=1.6707, tangentMode=RCTM_User, time=4, value=1}
         <at 4.1 0>
        ]}}}
     CurveIndex=$Emitter.LoopedAge
     Scale Curve=500
    )
    Use Spawn Probability=false
   ) {enabled=false, module=SpawnRate001}
   &SpawnRate(SpawnRate=100, Use Spawn Probability=false)
  ]
  particleSpawnScript [
   &InitializeParticle(
    Color=<rgba 1.7311 1.8926 10 0.5>
    Lifetime=&RandomRangeFloat(Minimum=0.2, Maximum=0.3)
    Sprite Size=&Multiply_Vector2D_ByFloat(
     Float=&Multiply_Float(A=&RandomRangeFloat(Minimum=1, Maximum=2), B=$User.Scale)
     Vector2D=<15 150>
    )
    Sprite Size Mode=::NewEnumerator4
    UsePositionOffset=false
    Write Material Random=true
   )
   &ShapeLocation(
    MaskSpawnGroup=false
    Override Seed=false
    Sphere Radius=&Multiply_Float(A=10, B=$User.Scale)
    UseOwnerScale=true
   )
   &AddVelocity(Module.Use Falloff Distance=false, Override Seed=false, Velocity Mode=::NewEnumerator1, Velocity Speed=3400)
  ]
  particleUpdateScript [
   &ParticleState()
   &ScaleSpriteSize(
    Scale Factor=&MakeVector2D(
     X=&FloatFromCurve(FloatCurve=DataInterface() {propertyValues {Curve {keys [
          {interpMode=RCIM_Cubic, leaveTangent=11.7691, tangentMode=RCTM_Break, time=0, value=0}
          {arriveTangent=0.0629, interpMode=RCIM_Cubic, leaveTangent=0.0629, tangentMode=RCTM_User, time=0.2, value=1}
          {interpMode=RCIM_Cubic, leaveTangent=0.027, tangentMode=RCTM_Break, time=0.35, value=1}
          {arriveTangent=0.0384, tangentMode=RCTM_User, time=1, value=0}
         ]}}})
     Y=1
    )
    Scale Sprite Size Mode=::NewEnumerator2
   )
   &SolveForcesAndVelocity(Clamp Velocity=false, Limit Acceleration=false)
   &ScaleSpriteSizeBySpeed(Max Scale Factor=<1 2>, Velocity Threshold=3500)
   &FN_Sprite_Attributes(
    Color=&MultiplyLinearColor(Linear Color=$Particles.Color, Scale Factor=<0.025 0.025 0.025 1>)
    Create Color Attribute=true
    Create SpriteSize Attribute=true
    SpriteSize=&Multiply_Vector2D(A=$Particles.SpriteSize, B=<5 3>)
    Write to Module Namespace=true
   )
   &ScaleColor(
    Scale Alpha=&FloatFromCurve(FloatCurve=DataInterface() {propertyValues {Curve {
        defaultValue=1
        keys [
         {arriveTangent=9.0725, interpMode=RCIM_Cubic, leaveTangent=9.0725, tangentMode=RCTM_User, time=0, value=0}
         {arriveTangent=-2.5249, interpMode=RCIM_Cubic, leaveTangent=-2.5249, tangentMode=RCTM_User, time=0.1408, value=0.9847}
         {arriveTangent=-0.0405, interpMode=RCIM_Cubic, leaveTangent=-0.0405, tangentMode=RCTM_User, time=1, value=0}
        ]
       }}})
    ScaleA=true
    ScaleRGB=true
   )
  ]
  rendererClasses [&NiagaraSpriteRendererProperties]
  renderers [
   {rendererClass=&NiagaraSpriteRendererProperties}
   {rendererClass=&NiagaraSpriteRendererProperties, rendererIndex=1}
  ]
 }
 Muzzle_Inner_Flares {
  emitterUpdateScript [
   &EmitterState(Delay First Loop Only=true, Life Cycle Mode=::NewEnumerator1, Loop Delay=0.2, Loop Duration=$System.CurrentLoopDuration, UseLoopDelay=true)
   &SpawnBurst_Instantaneous(Use Spawn Probability=false, UseLoopCountLimit=false)
   &SpawnRate(SpawnRate=22, Use Spawn Probability=false)
  ]
  particleSpawnScript [&InitializeParticle(
    Color=$User.PrimaryColor
    Lifetime Max=0.1
    Lifetime Min=0.05
    Lifetime Mode=::NewEnumerator1
    Sprite Rotation Mode=::NewEnumerator2
    Sprite Size Mode=::NewEnumerator1
    Uniform Sprite Size Max=700
    Uniform Sprite Size Min=400
    UsePositionOffset=false
    Write Material Random=true
   )]
  particleUpdateScript [
   &ParticleState()
   &CameraOffset(Camera Offset Amount=400)
   /SetVariables(
    Particles.Corona Color:LinearColor=&MakeLinearColorFromVectorAndFloat(
     "Vector (RGB)"=&MakeVectorFromLinearColorRGB(Color=<rgba 12.2533 19.1558 50 1>)
     "Float (Alpha)"=&FloatFromCurve(
      FloatCurve=DataInterface() {propertyValues {Curve {keys [
          <at 0 0 RCIM_Cubic>
          {arriveTangent=2.5, interpMode=RCIM_Cubic, leaveTangent=2.5, time=0.165, value=0.65}
          {arriveTangent=0.1786, interpMode=RCIM_Cubic, leaveTangent=0.1786, time=0.4, value=1}
          {arriveTangent=-1.6667, interpMode=RCIM_Cubic, leaveTangent=-1.6667, time=0.725, value=0.75}
          <at 1 0 RCIM_Cubic>
         ]}}}
      Scale Curve=0.5
     )
    )
    Particles.Corona Size:Vector2f=&Vector2DFromFloat(Value=&Multiply_Float(A=1350, B=$User.Scale))
    Particles.Small Corona Color:LinearColor=&MakeLinearColorFromVectorAndFloat(
     "Vector (RGB)"=&MakeVectorFromLinearColorRGB(Color=<rgba 65 7.6172 44.5952 1>)
     "Float (Alpha)"=&FloatFromCurve(
      FloatCurve=DataInterface() {propertyValues {Curve {keys [
          <at 0 0 RCIM_Cubic>
          {arriveTangent=2.5, interpMode=RCIM_Cubic, leaveTangent=2.5, time=0.165, value=0.65}
          {arriveTangent=0.1786, interpMode=RCIM_Cubic, leaveTangent=0.1786, time=0.4, value=1}
          {arriveTangent=-1.6667, interpMode=RCIM_Cubic, leaveTangent=-1.6667, time=0.725, value=0.75}
          <at 1 0 RCIM_Cubic>
         ]}}}
      Scale Curve=0.5
     )
    )
    Particles.Small Corona Size:Vector2f=&Vector2DFromFloat(Value=&Multiply_Float(A=450, B=$User.Scale))
   ) {isSetParameters=true}
   &DynamicMaterialParameters(
    Index 3 Param 1=&FloatFromCurve(
     FloatCurve=DataInterface() {propertyValues {Curve {keys [<at 0 0.4 RCIM_Cubic>,<at 0.5 0.2 RCIM_Cubic>,<at 1 0.4 RCIM_Cubic>]}}}
     CurveIndex=&Add_Float(A=$Particles.NormalizedAge, B=&RandomRangeFloat(Minimum=0, Maximum=1))
    )
    Index 3 Param 2=1
    Index 3 Param 3=&RandomRangeFloat(Minimum=0.25, Maximum=0.35)
    Index 3 Param 4=1
    Module.Param0WriteEnabled3=true
    Module.Param1WriteEnabled3=true
    Module.Param2WriteEnabled3=true
    Module.Param3WriteEnabled3=true
    Write Parameter Index 0=false
    Write Parameter Index 3=true
   )
  ]
  rendererClasses [&NiagaraSpriteRendererProperties]
  renderers [
   {rendererClass=&NiagaraSpriteRendererProperties}
   {rendererClass=&NiagaraSpriteRendererProperties, rendererIndex=1}
  ]
 }
 PostRadialBlur {
  emitterUpdateScript [
   &EmitterState(Inactive Response=::NewEnumerator2, Life Cycle Mode=::NewEnumerator1, Loop Behavior=::NewEnumerator1, Loop Duration=4, UseLoopDelay=false)
   &SpawnBurst_Instantaneous(Use Spawn Probability=false, UseLoopCountLimit=true)
  ]
  particleSpawnScript [&InitializeParticle(
    Mesh Scale Mode=::NewEnumerator3
    Sprite Size Mode=::NewEnumerator3
    Uniform Sprite Size=&Multiply_Float(A=3000, B=$User.Scale)
    UsePositionOffset=true
    Write Material Random=true
   )]
  particleUpdateScript [
   &ParticleState(Kill Particles When Lifetime Has Elapsed=false)
   /SetVariables(Particles.float:NiagaraFloat=&Waveform(Global Amplitude Scale=3, "Waveform [1]"=::NewEnumerator12, "[1] Frequency"=20, "[1] Amplitude Min/Max"=<0.5 1>)) {isSetParameters=true}
   &DynamicMaterialParameters(
    Index 0 Param 1=$Particles.float
    Index 0 Param 2=$Particles.float
    Index 0 Param 3=0.8
    Index 1 Param 1=0.2
    Index 1 Param 2=500
    Index 1 Param 3=0.2
    Index 2 Param 1=1
    Index 2 Param 2=1
    Index 2 Param 3=1
    Index 2 Param 4=0.5
    Index 3 Param 1=-3
    Index 3 Param 2=200
    Module.Index0Param0Write=true
    Module.Index0Param1Write=true
    Module.Index0Param2Write=true
    Module.Index0Param3Write=true
    Module.Index1Param0Write=true
    Module.Index1Param1Write=true
    Module.Index1Param2Write=true
    Module.Index1Param3Write=true
    Module.Index2Param0Write=true
    Module.Index2Param1Write=true
    Module.Index2Param2Write=true
    Module.Index2Param3Write=true
    Module.Index3Param0Write=true
    Module.Index3Param1Write=true
    Module.Index3Param2Write=true
    Module.Index3Param3Write=true
    Write Parameter Index 1=true
    Write Parameter Index 2=true
    Write Parameter Index 3=true
   )
   &ScaleColor(
    Scale RGB=&Multiply_VectorByFloat(
     Vector=&VectorFromFloat(Value=&Waveform("Waveform [1]"=::NewEnumerator12, "[1] Frequency"=1, "[1] Amplitude Min/Max"=<1 1>))
     Float=0.5
    )
    ScaleA=true
    ScaleRGB=true
   )
  ]
  rendererClasses [&NiagaraSpriteRendererProperties]
  renderers [{rendererClass=&NiagaraSpriteRendererProperties}]
 }
 Radial_Energy {
  emitterUpdateScript [
   &EmitterState(
    Inactive Response=::NewEnumerator2
    Life Cycle Mode=::NewEnumerator1
    Loop Behavior=::NewEnumerator1
    Loop Duration Mode=::NewEnumerator1
    UseLoopDelay=false
   ) {module=EmitterState001}
   &SpawnBurst_Instantaneous(Spawn Count=1, Use Spawn Probability=false)
  ]
  particleSpawnScript [&InitializeParticle(
    Color=<rgba 1 2 5 0.5>
    Lifetime=10
    Sprite Size Mode=::NewEnumerator3
    Uniform Sprite Size=&Multiply_Float(A=$User.Scale, B=1900)
    UsePositionOffset=false
    Write Material Random=true
   )]
  particleUpdateScript [
   &ParticleState(Kill Particles When Lifetime Has Elapsed=false)
   &FN_Radial_Material(Base Sprite Tile V=2, Depth Fade=250, Radial Mask Power=10, Radial Mask Tile U=0, Sprite Extrusion Speed=0.6, Sprite Spin Speed=0.1)
  ]
  rendererClasses [&NiagaraSpriteRendererProperties]
  renderers [{rendererClass=&NiagaraSpriteRendererProperties}]
 }
 SmallLightning {
  emitterUpdateScript [
   &EmitterState(
    Inactive Response=::NewEnumerator2
    Life Cycle Mode=::NewEnumerator1
    Loop Behavior=::NewEnumerator1
    Loop Duration Mode=::NewEnumerator1
    UseLoopDelay=false
   ) {module=EmitterState001}
   &SpawnRate(SpawnRate=20, Use Spawn Probability=false)
  ]
  particleSpawnScript [
   &InitializeParticle(
    Lifetime=0.1
    Mesh Scale=&Multiply_VectorByFloat(Vector=&RandomRangeVector(Minimum=<5 2.5 2.5>, Maximum=<6.5 7.5 7.5>), Float=$User.Scale)
    Mesh Scale Mode=::NewEnumerator4
    UsePositionOffset=false
    Write Material Random=true
   )
   &InitialMeshOrientation(
    Mesh Orientation Mode=::NewEnumerator2
    Rotation=&RandomRangeVector(Minimum=<0 0 0>, Maximum=<1 1 0>)
    Use Rotation Vector=true
   ) {module=InitialMeshOrientation001}
   &ShapeLocation(
    Cylinder Height=50
    Cylinder Radius=30
    MaskSpawnGroup=false
    Override Seed=false
    Shape Primitive=::NewEnumerator1
    Surface Only=false
    UseOwnerScale=true
   )
   /SetVariables(Particles.Mesh int32:NiagaraInt32=&RandomRangeInt(Minimum=0, Maximum=1)) {isSetParameters=true}
   &DynamicMaterialParameters(
    Index 0 Param 1=&Multiply_Float(A=10, B=$User.Scale)
    Index 0 Param 2=1
    Index 0 Param 3=3
    Index 0 Param 4=1
    Module.Param0WriteEnabled=true
    Module.Param1WriteEnabled=true
    Module.Param2WriteEnabled=true
    Module.Param3WriteEnabled=true
   )
   /SetVariables(Particles.DY2:Vector4f=&MakeVector4(X=3, Y=2, Z=&RandomRangeFloat(Minimum=2, Maximum=5), W=1)) {isSetParameters=true}
   &CameraOffset()
   &AddVelocity(
    Module.Use Falloff Distance=false
    Override Seed=false
    Velocity Mode=::NewEnumerator1
    Velocity Origin=$Particles.Position
    Velocity Speed=100
   )
  ]
  particleUpdateScript [
   &ParticleState()
   &ScaleColor(
    Scale RGB=&Multiply_VectorByFloat(
     Vector=<1 1 1>
     Float=&FloatFromCurve(FloatCurve=DataInterface() {propertyValues {Curve {keys [
          <at 0 0.5>
          <at 0.3 0.5>
          {arriveTangent=-0.8, interpMode=RCIM_Cubic, leaveTangent=-0.8, time=0.5, value=0.1}
          {arriveTangent=0.8, interpMode=RCIM_Cubic, leaveTangent=0.8, time=0.8, value=0.1}
          <at 1 0.5>
         ]}}})
    )
    ScaleA=true
    ScaleRGB=true
   )
   &SolveForcesAndVelocity(Clamp Velocity=false, Limit Acceleration=false)
  ]
  rendererClasses [&NiagaraMeshRendererProperties]
  renderers [
   {rendererClass=&NiagaraMeshRendererProperties, rendererIndex=0}
   {rendererClass=&NiagaraMeshRendererProperties, rendererIndex=1}
  ]
 }
}
summary {
 systemName=NS_PistolCanary_TentacleAttack_Whip_Proj
 emitters [
  {bEnabled=false, emitterName=BigBall, rendererClasses [&NiagaraMeshRendererProperties]}
  {emitterName=CoreFlare, rendererClasses [&NiagaraSpriteRendererProperties]}
  {emitterName=Glow, rendererClasses [&NiagaraSpriteRendererProperties, &NiagaraLightRendererProperties]}
  {emitterName=Radial_Energy, rendererClasses [&NiagaraSpriteRendererProperties]}
  {emitterName=Muzzle_Inner_Flares, rendererClasses [&NiagaraSpriteRendererProperties]}
  {emitterName=PostRadialBlur, rendererClasses [&NiagaraSpriteRendererProperties]}
  {emitterName=LoopSparks, rendererClasses [&NiagaraSpriteRendererProperties]}
  {emitterName=SmallLightning, rendererClasses [&NiagaraMeshRendererProperties]}
 ]
 userVariables [
  User.ChargeUpLength {type=NiagaraFloat, defaultValue=0, description=""}
  User.PrimaryColor {type=LinearColor, description="", defaultValue {struct=LinearColor, value=<rgba 1 1 1 1>}}
  User.Scale {type=NiagaraFloat, defaultValue=1, description=""}
 ]
}
